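# Additive, unlit, two-sided Niagara mesh material for conical energy in the StageCue skydiving contrail.
# Three panning smoke samples, offset and sped by dynamic parameters, blend along V and take particle color for emissive.
# Opacity combines UV gradient masks, near-camera fade and optional turbulence breakup, while time-of-day values scale intensity.
# A Large Cone switch alters emissive and adds inverse-Fresnel falloff; sine/cosine wiggle, driven by dynamic parameters, forms world position offset.
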Material /BRCosmetics/Athena/Effects/Particles/Skydiving/NS_Contrail_StageCue/Materials/M_Contrail_StageCue_ConalEnergy
graph {
 properties {
  bAutomaticallySetUsageInEditor=false
  bUsedWithNiagaraMeshParticles=true
  blendMode=BLEND_Additive
  shadingModel=MSM_Unlit
  twoSided=true
  usedShadingModels=Unlit
 }
 nodes [
  MaterialFunctionCall() {materialFunction=&Fresnel_Function}
  n1 = TextureCoordinate() {vTiling=2}
  n2 = DynamicParameter()
  n3 = Add(A=ComponentMask(None=@n1) {r=true}, B=@n2.Param1)
  n4 = ComponentMask(None=@n1) {g=true}
  n5 = NamedRerouteUsage()
  n6 = TextureSample(UVs=Panner(Coordinate=AppendVector(A=@n3, B=@n4), Time=@n5) {speedY=9}) {texture=/Game/Effects/Fort_Effects/Textures/Smoke/T_Puffy_Smoke_Trail_01}
  n7 = Constant() {r=15}
  n8 = TextureCoordinate() {vTiling=4}
  n9 = AppendVector(A=Add(A=ComponentMask(None=@n8) {r=true}, B=@n2.Param1), B=ComponentMask(None=@n8) {g=true})
  n10 = Multiply(
   A=TextureSample(UVs=Panner(Coordinate=@n9, Time=@n5) {speedY=12}) {texture=&T_Puffy_Smoke_Trail_01}.R
   B=Constant() {r=50}
  )
  n11 = TextureCoordinate()
  n12 = Saturate(None=Power(Base=ComponentMask(None=@n11) {g=true}) {constExponent=8})
  n13 = ComponentMask(None=TextureCoordinate()) {g=true}
  n14 = TextureCoordinate()
  n15 = ComponentMask(None=@n14) {r=true}
  n16 = Panner(Coordinate=AppendVector(A=Add(A=@n15, B=@n2.Param1), B=ComponentMask(None=@n14) {g=true}), Time=NamedRerouteUsage()) {speedY=5}
  n17 = LinearInterpolate(
   A=LinearInterpolate(A=Multiply(A=@n6.R, B=@n7), B=Multiply(A=@n10, B=@n12), Alpha=@n13)
   B=Multiply(A=TextureSample(UVs=@n16) {texture=&T_Puffy_Smoke_Trail_01}.R, B=Constant() {r=3})
   Alpha=Saturate(None=Power(Base=@n13) {constExponent=5})
  )
  n18 = TextureSample(UVs=MaterialFunctionCall(UVs=TextureCoordinate(), "Rotation Angle (0-1)"=Constant() {r=0.5}) {materialFunction=&CustomRotator}.Rotated Values) {texture=/Game/Effects/Fort_Effects/Textures/Masks/T_Vert_Gradient_Softer}
  n19 = StaticSwitchParameter(True=Add(A=Multiply(A=@n17) {constB=0.5}, B=@n18.R), False=@n17) {parameterName=Large Cone}
  n20 = ParticleColor()
  NamedRerouteDeclaration(None=MaterialFunctionCall(Emissive=Multiply(A=@n19, B=@n20.RGB)) {materialFunction=&MF_FX_EmissiveAdjustments}.Result) {name=Emissive, nodeColor=<rgba 1 0 0.6562 1>}
  n21 = ScalarParameter() {defaultValue=1, parameterName=Pan Speed}
  NamedRerouteDeclaration(None=Multiply(A=Multiply(A=@n21, B=@n2.Param2), B=Time())) {name=PanSpeed, nodeColor=<rgba 0.2188 0 1 1>}
  n22 = NamedRerouteUsage()
  n23 = ComponentMask(None=TextureCoordinate()) {r=true}
  n24 = OneMinus(None=@n23)
  n25 = Power(Base=Multiply(A=@n23) {constB=5}) {constExponent=1}
  n26 = Multiply(A=Multiply(A=Power(Base=@n24) {constExponent=5}, B=Saturate(None=@n25))) {constB=0.25}
  n27 = ComponentMask(None=TextureCoordinate()) {r=true}
  n28 = Multiply(A=@n27) {constB=8}
  n29 = Power(Base=OneMinus(None=@n27)) {constExponent=5}
  n30 = Saturate(None=Saturate(None=Multiply(A=@n28, B=@n29)))
  n31 = Multiply(A=Multiply(A=Constant() {r=0.25}, B=@n30), B=Constant() {r=0.245})
  n32 = MaterialFunctionCall(Distance=ScalarParameter() {defaultValue=50, parameterName=Camera Fade Distance}) {materialFunction=&Near_Camera_Fade}
  n33 = Add(A=Multiply(A=Add(A=@n26, B=@n31), B=Multiply(A=@n30, B=@n32.Result)))
  n34 = Multiply(A=Saturate(None=Multiply(A=@n22, B=@n33)), B=@n18.RGB)
  n35 = TextureCoordinate() {vTiling=0.5}
  n36 = NamedRerouteUsage()
  n37 = Power(Base=TextureSample(UVs=Panner(Coordinate=@n35, Time=@n36) {speedY=2}) {samplerType=SAMPLERTYPE_Grayscale, texture=/Game/Effects/Fort_Effects/Textures/Plasma/T_Turbulence_Stretched_Balanced}.R) {constExponent=1.25}
  NamedRerouteDeclaration(None=StaticSwitchParameter(True=Multiply(A=@n34, B=@n37), False=@n34) {parameterName=Noise Breakup}) {name=Opacity, nodeColor=<rgba 0 0.9766 1 1>}
  n38 = DynamicParameter() {paramNames [WiggleOffset, WiggleRate, WiggleAmplitude, Rotation]}
  n39 = Add(A=Time(), B=@n38.WiggleOffset) {constB=0}
  n40 = Multiply(A=@n39, B=@n38.WiggleRate)
  n41 = Multiply(A=Constant3Vector() {constant=<rgba 3 0 0 0>}, B=Multiply(A=Sine(None=@n40), B=@n38.WiggleAmplitude) {constB=0.9}) {constB=0}
  n42 = Multiply(A=Multiply(A=Cosine(None=@n40) {period=1.2}, B=@n38.WiggleAmplitude) {constB=0.9}, B=Constant3Vector() {constant=<rgba 0 3 0 0>})
  n43 = Multiply(A=Sine(None=Multiply(A=@n39) {constB=1.1}) {period=-0.8}, B=@n38.WiggleAmplitude)
  n44 = Constant3Vector() {constant=<rgba 0 0 1 0>}
  NamedRerouteDeclaration(None=Add(A=Add(A=@n41, B=@n42), B=Multiply(A=@n43, B=@n44))) {name=WPO, nodeColor=<rgba 1 0 0.4219 1>}
  NamedRerouteDeclaration(None=@n20.A) {name=Alpha Output, nodeColor=<rgba 1 0 0.1875 1>}
  n45 = ScalarParameter() {defaultValue=0.75, parameterName=Morning Value}
  n46 = NamedRerouteUsage()
  n47 = ScalarParameter() {defaultValue=1, parameterName=Day Value}
  n48 = ScalarParameter() {defaultValue=0.75, parameterName=Evening Value}
  n49 = ScalarParameter() {defaultValue=0.5, parameterName=Night Value}
  NamedRerouteDeclaration(None=MaterialFunctionCall(
    Value_Morning=Multiply(A=@n45, B=@n46)
    Value_Day=Multiply(A=@n47, B=@n46)
    Value_Evening=Multiply(A=@n48, B=@n46)
    Value_Night=Multiply(A=@n49, B=@n46)
   ) {materialFunction=&Scalar_Time_of_Day}.Result) {name=ToD Output, nodeColor=<rgba 0.6406 1 0 1>}
  NamedRerouteUsage()
  NamedRerouteUsage()
  n50 = NamedRerouteUsage()
  n51 = Fresnel() {exponent=0.5}
  StaticSwitchParameter(True=Multiply(A=@n50, B=OneMinus(None=@n51)), False=@n50) {parameterName=Large Cone}
 ]
}
