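# Translucent unlit Niagara ribbon material for the cue ball of the Stage Cue skydiving contrail, drawing a colored, scrolling smoke trail.
# Color blends two gradients along U, selected by a dynamic parameter, with particle color and an emissive adjustment function.
# Opacity layers one smoke texture at three tilings and pan speeds, blended and masked along U with near-camera fade.
# A faint ramp-masked cloud layer is added, then particle alpha scales the result.
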Material /BRCosmetics/Athena/Effects/Particles/Skydiving/NS_Contrail_StageCue/Materials/M_Contrail_StageCue_CueBall_Ribbon
graph {
 properties {
  bEnableResponsiveAA=true
  bOutputTranslucentVelocity=true
  bUsedWithNiagaraRibbons=true
  bUsedWithNiagaraSprites=true
  blendMode=BLEND_Translucent
  pixelDepthOffsetMode=PDOM_Legacy
  shadingModel=MSM_Unlit
  usedShadingModels=Unlit
 }
 nodes [
  n1 = ScalarParameter() {defaultValue=1, parameterName=Pan Speed}
  n2 = NamedRerouteUsage()
  NamedRerouteDeclaration(None=Multiply(A=Multiply(A=@n1, B=@n2), B=Time())) {name=PanSpeed, nodeColor=<rgba 0.2188 0 1 1>}
  n3 = MaterialFunctionCall() {materialFunction=&LinearGradient}
  n4 = Power(Base=@n3.UGradient)
  n5 = ScalarParameter() {defaultValue=5, parameterName=GradientPush_A}
  n6 = LinearInterpolate(
   A=Constant3Vector() {constant=<rgba 1 0.8141 0.1253 1>}
   B=Constant3Vector() {constant=<rgba 1 0.9185 0 1>}
   Alpha=Saturate(None=Multiply(A=@n4, B=@n5))
  )
  n7 = Constant3Vector() {constant=<rgba 1 0.3874 0 1>}
  n8 = MaterialFunctionCall() {materialFunction=&LinearGradient}
  n9 = Power(Base=@n8.UGradient)
  n10 = ScalarParameter() {defaultValue=5, parameterName=GradientPush_A}
  n11 = LinearInterpolate(
   A=Constant3Vector() {constant=<rgba 0.2945 0.8631 1 1>}
   B=Constant3Vector() {constant=<rgba 0.961 1 0.6562 1>}
   Alpha=Saturate(None=Multiply(A=@n9, B=@n10))
  )
  n12 = Constant3Vector() {constant=<rgba 1 0.3874 0 1>}
  NamedRerouteDeclaration(None=LinearInterpolate(
    A=LinearInterpolate(A=@n6, B=@n7, Alpha=@n3.UGradient)
    B=LinearInterpolate(A=@n11, B=@n12, Alpha=@n8.UGradient)
    Alpha=DynamicParameter() {defaultValue=<rgba 1 0 1 0>, paramNames [None, None, None, U offset]}.G
   ) {constAlpha=0}) {name=Color Gradient, nodeColor=<rgba 0.1484 0 1 1>}
  n13 = DynamicParameter()
  NamedRerouteDeclaration(None=@n13.Param2) {name=DP1_P2, nodeColor=<rgba 1 0 0.5156 1>}
  NamedRerouteDeclaration(None=@n13.Param1) {name=DP1_P1, nodeColor=<rgba 0 0.8359 1 1>}
  n14 = NamedRerouteUsage()
  n15 = ParticleColor()
  NamedRerouteDeclaration(None=MaterialFunctionCall(Emissive=Multiply(A=@n14, B=@n15.RGB)) {materialFunction=&MF_FX_EmissiveAdjustments}.Result) {name=Final Emissive, nodeColor=<rgba 1 0.1875 0 1>}
  n16 = ComponentMask(None=TextureCoordinate()) {r=true}
  n17 = Power(Base=OneMinus(None=@n16)) {constExponent=5}
  n18 = Saturate(None=Power(Base=Multiply(A=@n16) {constB=5}) {constExponent=1})
  n19 = TextureCoordinate() {uTiling=4}
  n20 = AppendVector(A=Add(A=ComponentMask(None=@n19) {r=true}, B=NamedRerouteUsage()), B=ComponentMask(None=@n19) {g=true})
  n21 = NamedRerouteUsage()
  n22 = TextureSample(UVs=MaterialFunctionCall(UVs=Panner(Coordinate=@n20, Time=@n21) {speedX=-9}, "Rotation Angle (0-1)"=Constant() {r=0.75}) {materialFunction=&CustomRotator}.Rotated Values) {texture=/Game/Effects/Fort_Effects/Textures/Smoke/T_Puffy_Smoke_Trail_01}
  n23 = Constant() {r=15}
  n24 = TextureCoordinate() {uTiling=8}
  n25 = ComponentMask(None=@n24) {r=true}
  n26 = NamedRerouteUsage()
  n27 = Panner(Coordinate=AppendVector(A=Add(A=@n25, B=@n26), B=ComponentMask(None=@n24) {g=true}), Time=@n21) {speedX=-12}
  n28 = Constant() {r=0.75}
  n29 = Multiply(
   A=TextureSample(UVs=MaterialFunctionCall(UVs=@n27, "Rotation Angle (0-1)"=@n28) {materialFunction=&CustomRotator}.Rotated Values) {texture=&T_Puffy_Smoke_Trail_01}.R
   B=Constant() {r=50}
  )
  n30 = ComponentMask(None=TextureCoordinate()) {r=true}
  n31 = Saturate(None=Power(Base=OneMinus(None=@n30)) {constExponent=8})
  n32 = TextureCoordinate() {uTiling=2}
  n33 = Add(A=ComponentMask(None=@n32) {r=true}, B=NamedRerouteUsage())
  n34 = ComponentMask(None=@n32) {g=true}
  n35 = MaterialFunctionCall(UVs=Panner(Coordinate=AppendVector(A=@n33, B=@n34), Time=NamedRerouteUsage()) {speedX=-5}, "Rotation Angle (0-1)"=Constant() {r=0.75}) {materialFunction=&CustomRotator}
  n36 = ComponentMask(None=TextureCoordinate()) {r=true}
  n37 = LinearInterpolate(
   A=LinearInterpolate(A=Multiply(A=@n22.R, B=@n23), B=Multiply(A=@n29, B=@n31), Alpha=Saturate(None=@n17))
   B=Multiply(A=TextureSample(UVs=@n35.Rotated Values) {texture=&T_Puffy_Smoke_Trail_01}.R, B=Constant() {r=3})
   Alpha=@n36
  )
  n38 = Multiply(A=@n36) {constB=8}
  n39 = Power(Base=OneMinus(None=@n36)) {constExponent=5}
  n40 = Saturate(None=Saturate(None=Multiply(A=@n38, B=@n39)))
  n41 = Multiply(A=Constant() {r=0.25}, B=@n40)
  n42 = Add(A=Multiply(A=Multiply(A=@n17, B=@n18)) {constB=0.25}, B=Multiply(A=Add(A=@n37, B=@n41), B=Constant() {r=0.245}))
  n43 = Multiply(
   A=@n40
   B=MaterialFunctionCall(Distance=ScalarParameter() {defaultValue=50, parameterName=Camera Fade Distance}) {materialFunction=&Near_Camera_Fade}.Result
  )
  n44 = TextureCoordinate()
  n45 = Constant() {r=0.25}
  n46 = Multiply(A=TextureSample(UVs=MaterialFunctionCall(UVs=@n44, "Rotation Angle (0-1)"=@n45) {materialFunction=&CustomRotator}.Rotated Values) {texture=/Game/Effects/Fort_Effects/Textures/Ramps/T_Power_Source_Sphere_Ramp_Soft}.R) {constB=0.1}
  n47 = TextureSample(UVs=Panner(Coordinate=TextureCoordinate() {uTiling=0.5}) {speedX=-2}) {texture=/Game/Effects/Fort_Effects/Textures/Smoke/T_Clouds_01}
  NamedRerouteDeclaration(None=Add(A=Multiply(A=@n42, B=@n43), B=Multiply(A=@n46, B=@n47.R))) {name=Noise, nodeColor=<rgba 0 0.1328 1 1>}
  n48 = NamedRerouteUsage()
  NamedRerouteDeclaration(None=Saturate(None=Multiply(A=@n15.A, B=@n48))) {name=Final Alpha, nodeColor=<rgba 0.4531 1 0 1>}
  NamedRerouteUsage()
  NamedRerouteUsage()
 ]
}
