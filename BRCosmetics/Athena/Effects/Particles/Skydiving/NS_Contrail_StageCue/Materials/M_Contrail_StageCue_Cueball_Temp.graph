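# Temporary-named Default Lit cue ball material for Niagara mesh particles in the StageCue skydiving contrail effect.
# Niagara dynamic parameters offset time and drive wiggle rate and amplitude for sine and cosine waves.
# These build a three-axis vector, weighted toward X and Y, with Z running at a fixed rate independent of WiggleRate.
# The cue ball texture is sampled separately, and the Rotation parameter is unused.
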Material /BRCosmetics/Athena/Effects/Particles/Skydiving/NS_Contrail_StageCue/Materials/M_Contrail_StageCue_Cueball_Temp
graph {
 properties {bAutomaticallySetUsageInEditor=false, bUsedWithNiagaraMeshParticles=true, usedShadingModels=Default Lit}
 nodes [
  n1 = Constant3Vector() {constant=<rgba 3 0 0 0>}
  n2 = DynamicParameter() {paramNames [WiggleOffset, WiggleRate, WiggleAmplitude, Rotation]}
  n3 = Add(A=Time(), B=@n2.WiggleOffset) {constB=0}
  n4 = Multiply(A=@n3, B=@n2.WiggleRate)
  n5 = Multiply(A=Sine(None=@n4), B=@n2.WiggleAmplitude) {constB=0.9}
  n6 = Multiply(A=Cosine(None=@n4) {period=1.2}, B=@n2.WiggleAmplitude) {constB=0.9}
  n7 = Constant3Vector() {constant=<rgba 0 3 0 0>}
  n8 = Sine(None=Multiply(A=@n3) {constB=1.1}) {period=-0.8}
  Add(
   A=Add(A=Multiply(A=@n1, B=@n5) {constB=0}, B=Multiply(A=@n6, B=@n7))
   B=Multiply(A=Multiply(A=@n8, B=@n2.WiggleAmplitude), B=Constant3Vector() {constant=<rgba 0 0 1 0>})
  )
  TextureSample() {texture=/BRCosmetics/Athena/Effects/Particles/Skydiving/NS_Contrail_StageCue/Textures/T_FX_CueBall}
 ]
}
