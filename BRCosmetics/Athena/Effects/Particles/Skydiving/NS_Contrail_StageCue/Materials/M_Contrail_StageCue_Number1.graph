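# Unlit translucent material for Niagara ribbons and mesh particles in the StageCue skydiving contrail, built around the Number1 texture.
# Panned Number1, streak and wispy-noise textures combine with a soft-edged UV ribbon mask to form the trail.
# Two warm and cyan-to-orange gradients, blended by a dynamic parameter, are tinted by particle color for emissive.
# Alpha adds end masks, optional trailing fade, depth fade and near-camera fade, switched by shading path and quality.
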Material /BRCosmetics/Athena/Effects/Particles/Skydiving/NS_Contrail_StageCue/Materials/M_Contrail_StageCue_Number1
graph {
 properties {
  bUsedWithNiagaraMeshParticles=true
  bUsedWithNiagaraRibbons=true
  blendMode=BLEND_Translucent
  pixelDepthOffsetMode=PDOM_Legacy
  shadingModel=MSM_Unlit
  usedShadingModels=Unlit
 }
 nodes [
  n1 = Multiply(A=NamedRerouteUsage(), B=Constant() {r=3})
  n2 = NamedRerouteUsage()
  n3 = Multiply(A=Add(A=Power(Base=@n1) {constExponent=1.3}, B=NamedRerouteUsage()), B=Multiply(A=Saturate(None=@n2), B=Constant() {r=0.5}))
  n4 = Multiply(A=NamedRerouteUsage()) {constB=0.25}
  NamedRerouteDeclaration(None=Saturate(None=Add(A=@n3, B=@n4))) {name=Emissive Output, nodeColor=<rgba 0 1 0.2969 1>}
  n5 = MaterialFunctionCall() {materialFunction=&LinearGradient}
  n6 = Power(Base=@n5.UGradient)
  n7 = ScalarParameter() {defaultValue=5, parameterName=GradientPush_A}
  n8 = LinearInterpolate(
   A=Constant3Vector() {constant=<rgba 1 0.8141 0.1253 1>}
   B=Constant3Vector() {constant=<rgba 1 0.9185 0 1>}
   Alpha=Saturate(None=Multiply(A=@n6, B=@n7))
  )
  n9 = Constant3Vector() {constant=<rgba 1 0.3874 0 1>}
  n10 = MaterialFunctionCall() {materialFunction=&LinearGradient}
  n11 = Power(Base=@n10.UGradient)
  n12 = ScalarParameter() {defaultValue=5, parameterName=GradientPush_A}
  n13 = LinearInterpolate(
   A=Constant3Vector() {constant=<rgba 0.2945 0.8631 1 1>}
   B=Constant3Vector() {constant=<rgba 0.961 1 0.6562 1>}
   Alpha=Saturate(None=Multiply(A=@n11, B=@n12))
  )
  n14 = Constant3Vector() {constant=<rgba 1 0.3874 0 1>}
  NamedRerouteDeclaration(None=LinearInterpolate(
    A=LinearInterpolate(A=@n8, B=@n9, Alpha=@n5.UGradient)
    B=LinearInterpolate(A=@n13, B=@n14, Alpha=@n10.UGradient)
    Alpha=DynamicParameter() {defaultValue=<rgba 1 0 1 0>, paramNames [None, None, None, U offset]}.G
   ) {constAlpha=0}) {name=ColorGradient, nodeColor=<rgba 0.1484 0 1 1>}
  n15 = ScalarParameter() {defaultValue=1, parameterName=Speed Direction}
  n16 = DynamicParameter() {parameterIndex=2, defaultValue=<rgba 1 1 1 0>, paramNames [Speed Dir Mult, None, None, None]}
  NamedRerouteDeclaration(None=Multiply(A=Time(), B=Multiply(A=@n15, B=@n16.Speed Dir Mult))) {name=PanSpeed, nodeColor=<rgba 1 0.1172 0 1>}
  n17 = TextureCoordinate()
  n18 = OneMinus(None=ComponentMask(None=@n17) {r=true})
  NamedRerouteDeclaration(None=@n18) {name=U Grad Invert, nodeColor=<rgba 0.1953 0 1 1>}
  n19 = ComponentMask(None=@n17) {g=true}
  n20 = Multiply(A=@n19, B=OneMinus(None=@n19))
  n21 = Saturate(None=Power(Base=Multiply(A=@n20) {constB=10}) {constExponent=3})
  NamedRerouteDeclaration(None=Multiply(A=@n18, B=@n21)) {name=Main Ribbon Shape, nodeColor=<rgba 0.125 1 0 1>}
  n22 = TextureCoordinate() {uTiling=4}
  n23 = ScalarParameter() {defaultValue=4, parameterName=UV Mult}
  n24 = Multiply(A=ComponentMask(None=@n22) {r=true}, B=@n23)
  n25 = OneMinus(None=ComponentMask(None=@n22) {g=true})
  n26 = NamedRerouteUsage()
  n27 = TextureSample(UVs=Panner(Coordinate=AppendVector(A=@n24, B=@n25), Time=@n26) {speedX=-3}) {texture=/BRCosmetics/Athena/Effects/Particles/Skydiving/NS_Contrail_StageCue/Textures/T_Contrail_StageCue_Number1}
  n28 = TextureCoordinate() {vTiling=0.25}
  n29 = Multiply(
   A=Power(Base=@n27.R) {constExponent=5}
   B=TextureSample(UVs=Panner(Coordinate=@n28) {speedY=1}) {texture=/Game/Effects/Fort_Effects/Textures/Energy/T_OutpostShockwaveStreaks_02}.R
  )
  n30 = Constant() {r=50}
  NamedRerouteDeclaration(None=Add(A=@n27.R, B=Multiply(A=@n29, B=@n30))) {name=Main Shape, nodeColor=<rgba 1 0 0.4453 1>}
  n31 = TextureCoordinate() {uTiling=0.25}
  n32 = AppendVector(A=Multiply(A=ComponentMask(None=@n31) {r=true}, B=@n23), B=OneMinus(None=ComponentMask(None=@n31) {g=true}))
  n33 = Multiply(
   A=TextureSample(UVs=Panner(Coordinate=@n32, Time=@n26) {speedX=-1}) {samplerType=SAMPLERTYPE_Grayscale, texture=/Game/Effects/Fort_Effects/Textures/Fire/T_WispyFireNoise_Soft_M}.R
   B=ScalarParameter() {defaultValue=5, parameterName=1's Mult}
  )
  n34 = Saturate(None=Power(Base=NamedRerouteUsage()) {constExponent=8})
  NamedRerouteDeclaration(None=Multiply(A=Multiply(A=@n33, B=@n34), B=NamedRerouteUsage())) {name=Noise, nodeColor=<rgba 0 0.7656 1 1>}
  NamedRerouteDeclaration(None=Multiply(A=@n21) {constB=0.25}) {name=Soft Sides, nodeColor=<rgba 1 0.9141 0 1>}
  n35 = NamedRerouteUsage()
  n36 = Multiply(A=Constant() {r=0.05}, B=NamedRerouteUsage())
  n37 = Saturate(None=Add(A=@n3, B=Add(A=@n35, B=@n36)))
  n38 = ComponentMask(None=NamedRerouteUsage()) {r=true}
  n39 = Saturate(None=Multiply(A=Power(Base=@n38), B=Constant() {r=50}) {constB=10})
  n40 = Constant()
  n41 = Constant() {r=0.33}
  n42 = OneMinus(None=@n38)
  n43 = DynamicParameter() {parameterIndex=1, defaultValue=<rgba 1 1 0 0>, paramNames [None, None, BackOffset, Use Trailing Fade]}
  n44 = LinearInterpolate(A=@n39, B=Saturate(None=SmoothStep(Min=@n40, Max=@n41, Value=@n42)), Alpha=@n43.Use Trailing Fade) {constA=1}
  NamedRerouteDeclaration(None=Multiply(A=Multiply(A=@n37, B=@n44), B=@n39)) {name=Final Trail Shape, nodeColor=<rgba 0 0.7891 1 1>}
  n45 = NamedRerouteUsage()
  n46 = ParticleColor()
  n47 = Multiply(A=NamedRerouteUsage(), B=@n46.RGB)
  NamedRerouteDeclaration(None=MaterialFunctionCall(Emissive=Multiply(A=@n45, B=@n47)) {materialFunction=&MF_FX_EmissiveAdjustments}.Result) {name=Final Emissive, nodeColor=<rgba 0.8047 1 0 1>}
  n48 = NamedRerouteUsage()
  n49 = Multiply(A=Subtract(A=ComponentMask(None=@n48) {r=true}, B=ScalarParameter() {parameterName=Lead Mask Offset}) {constB=0}) {constB=5}
  n50 = Multiply(A=@n46.A, B=Multiply(A=NamedRerouteUsage(), B=Saturate(None=@n49)))
  n51 = DepthFade(Opacity=@n50, FadeDistance=ScalarParameter() {defaultValue=5, parameterName=DepthFade})
  n52 = Power(
   Base=MaterialFunctionCall(Distance=ScalarParameter() {defaultValue=100, parameterName=NearCamFadeDistance}) {materialFunction=&Near_Camera_Fade}.Result
   Exp=ScalarParameter() {defaultValue=2, parameterName=NearCamFadePower}
  )
  n53 = Constant() {r=1}
  NamedRerouteDeclaration(None=Multiply(
    A=ShadingPathSwitch(Default=@n51, Deferred=@n51, Forward=@n50, Mobile=@n50)
    B=QualitySwitch(Default=@n52, Low=@n53, High=@n52, Medium=@n53, Epic=@n52)
   )) {name=Final Alpha, nodeColor=<rgba 0 1 0.5547 1>}
  n54 = MaterialFunctionCall(Float2=TextureCoordinate()) {materialFunction=&BreakOutFloat2Components}
  NamedRerouteDeclaration(None=MaterialFunctionCall(X=OneMinus(None=@n54.R), Y=@n54.G) {materialFunction=&MakeFloat2}.Result) {name=UV's, nodeColor=<rgba 0.7812 1 0 1>}
  NamedRerouteUsage()
  NamedRerouteUsage()
 ]
}
