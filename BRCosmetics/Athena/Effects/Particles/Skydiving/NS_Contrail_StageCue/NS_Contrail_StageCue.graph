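# Skydiving contrail cosmetic that floats nine cue-ball meshes around the player, with a fire cone and ribbon trails.
# FN_Contrail_Setup's Dive Value blends ball formation between two position arrays and drives spin, color, opacity and ribbon width.
# Ribbons are fixed bursts re-placed each frame from the sampled skeletal bone along contrail velocity, sharing the balls' front-end preview offset.
# Several generic gizmo, particle and skeletal-mesh emitters are disabled.
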NiagaraSystem /BRCosmetics/Athena/Effects/Particles/Skydiving/NS_Contrail_StageCue/NS_Contrail_StageCue
dependencies {
 usedDataInterfaces [
  &NiagaraDataInterfaceCurve
  &NiagaraDataInterfaceArrayFloat3
  &NiagaraDataInterfaceColorCurve
  &NiagaraDataInterfaceArrayFloat
 ]
 usedDynamicInputs []
 usedModules [
  &self:SystemSpawnScript.SystemScriptSource.SystemScriptGraph.NiagaraNodeAssignment_3.SetVariables_#1
  &SampleSkeletalMesh
  &SystemState
  /Game/Effects/Niagara/Modules/Contrails/FN_Contrail_Setup
  &FN_Integrator
  &Timeline
 ]
}
emitters {
 CueBalls {
  emitterUpdateScript [
   &EmitterState(Life Cycle Mode=::NewEnumerator1, Loop Behavior=::NewEnumerator1, Loop Duration Mode=::NewEnumerator1, UseLoopDelay=false)
   &SpawnBurst_Instantaneous(Spawn Count=9, Use Spawn Probability=false, UseLoopCountLimit=false)
  ]
  particleSpawnScript [
   &PartitionParticles(Number Of Partitions=9)
   &InitializeParticle(
    Color Mode=::NewEnumerator0
    Lifetime=$Emitter.CurrentLoopDuration
    Mesh Scale Mode=::NewEnumerator3
    Position=$System.SampleSkeletalMesh.SampledPosition
    Position Mode=::NewEnumerator1
    Position Offset=&SelectVectorFromArray(Direct Array Index=$Particles.PartitionParticles.Partition, Vector Selection Array=DataInterface() {propertyValues {FloatData []}})
    Sprite Size Mode=::NewEnumerator3
    Uniform Sprite Size=35
    UsePositionOffset=false
    Write Material Random=true
   )
   &ApplyInitialForces()
   &InitialMeshOrientation(Use Rotation Vector=false)
  ]
  particleUpdateScript [
   &MeshRotationForce(Rotation=&Lerp_Vector(A=<0 0 0>, B=<0 50 0>, Alpha=$System.FN_Contrail_Setup.Dive Value)) {enabled=false, module=MeshRotationForce001}
   &ParticleState()
   &UpdateMeshOrientation(
    Rotation Coordinate Space=::NewEnumerator0
    Rotation Rate=&Lerp_Float(
     A=&RandomRangeFloat(Minimum=0.25, Maximum=0.35)
     B=&RandomRangeFloat(Minimum=1.5, Maximum=3)
     Alpha=&FloatFromCurve(
      FloatCurve=DataInterface() {propertyValues {Curve {keys [{interpMode=RCIM_Cubic, tangentMode=RCTM_Break, time=0, value=0}, {interpMode=RCIM_Cubic, tangentMode=RCTM_Break, time=0.5, value=1}]}}}
      CurveIndex=$System.FN_Contrail_Setup.Dive Value
     )
    )
    Rotation Vector=&RandomRangeVector(Minimum=<-1 -1 -1>, Maximum=<1 1 1>)
   )
   &SkeletalMeshLocation(
    Bone / Socket Index=4
    Bone Sampling Mode=::NewEnumerator4
    MaskSpawnGroup=false
    Orientation Sampling=::NewEnumerator1
    Sampled Position Offset=<0 0 150>
    Skeletal Mesh=$User.SkeletalMesh
   )
   /SetVariables(Particles.Position=&AddVectorToPosition(
     Position=&Add_Vector(
      A=$Particles.SkeletalMeshLocation.SampledPosition
      B=&RotateVectorByQuaternion(
       Vector To Rotate=&Lerp_Vector(
        A=&SelectVectorFromArray(Direct Array Index=$Particles.PartitionParticles.Partition, Direct Array Mode=::NewEnumerator1, Vector Selection Array=$System.Array1)
        B=&SelectVectorFromArray(Direct Array Index=$Particles.PartitionParticles.Partition, Direct Array Mode=::NewEnumerator1, Vector Selection Array=$System.Array2)
        Alpha=&FloatFromCurve(
         FloatCurve=DataInterface() {propertyValues {Curve {keys [
             {arriveTangent=6.8318, interpMode=RCIM_Cubic, leaveTangent=6.8318, tangentMode=RCTM_User, time=0.075, value=0}
             {arriveTangent=-2.3665, interpMode=RCIM_Cubic, leaveTangent=-2.3665, tangentMode=RCTM_User, time=0.35, value=1.875}
             {arriveTangent=-1.8964, interpMode=RCIM_Cubic, leaveTangent=-1.8964, tangentMode=RCTM_User, time=0.7, value=0.875}
             {interpMode=RCIM_Cubic, tangentMode=RCTM_User, time=0.9, value=1.25}
            ]}}}
         CurveIndex=$System.FN_Contrail_Setup.Dive Value
        )
       )
       Quaternion=&MultiplyQuaternion(Quaternion A=$Particles.SkeletalMeshLocation.SampledRotation, Quaternion B=<0 0 0 1>)
      )
     )
     Vector=&MakeCustomVectorFromBool(
      Bool=$User.bIsFrontEndPreview
      True Vector=&Lerp_Vector(
       A=<0 10 50>
       B=<0 -40 50>
       Alpha=&FloatFromCurve(
        FloatCurve=DataInterface() {propertyValues {Curve {keys [
            {arriveTangent=4.1785, interpMode=RCIM_Cubic, leaveTangent=4.1785, tangentMode=RCTM_User, time=0, value=0}
            {arriveTangent=1.1179, interpMode=RCIM_Cubic, leaveTangent=1.1179, tangentMode=RCTM_User, time=0.2042, value=0.6228}
            {arriveTangent=0.0224, interpMode=RCIM_Cubic, leaveTangent=0.0224, tangentMode=RCTM_User, time=1, value=1}
           ]}}}
        CurveIndex=$System.FN_Contrail_Setup.Dive Value
       )
      )
      False Vector=&Lerp_Vector(
       A=<0 10 50>
       B=<0 50 80>
       Alpha=&FloatFromCurve(
        FloatCurve=DataInterface() {propertyValues {Curve {keys [
            {arriveTangent=4.1785, interpMode=RCIM_Cubic, leaveTangent=4.1785, tangentMode=RCTM_User, time=0, value=0}
            {arriveTangent=1.1179, interpMode=RCIM_Cubic, leaveTangent=1.1179, tangentMode=RCTM_User, time=0.2042, value=0.6228}
            {arriveTangent=0.0224, interpMode=RCIM_Cubic, leaveTangent=0.0224, tangentMode=RCTM_User, time=1, value=1}
           ]}}}
        CurveIndex=$System.FN_Contrail_Setup.Dive Value
       )
      )
     )
    )) {isSetParameters=true, module=SetVariables_#2}
   &DynamicMaterialParameters(
    Index 0 Param 1=&RandomRangeFloat(Minimum=0, Maximum=1)
    Index 0 Param 2=&Lerp_Float(
     A=0.75
     B=1.5
     Alpha=&FloatFromCurve(
      FloatCurve=DataInterface() {propertyValues {Curve {keys [{interpMode=RCIM_Cubic, tangentMode=RCTM_Break, time=0.05, value=0}, {interpMode=RCIM_Cubic, tangentMode=RCTM_Break, time=0.75, value=1}]}}}
      CurveIndex=$System.FN_Contrail_Setup.Dive Value
     )
    )
    Index 0 Param 3=&Lerp_Float(
     A=0.25
     B=0.3
     Alpha=&FloatFromCurve(
      FloatCurve=DataInterface() {propertyValues {Curve {keys [{interpMode=RCIM_Cubic, tangentMode=RCTM_Break, time=0, value=0}, {interpMode=RCIM_Cubic, tangentMode=RCTM_Break, time=0.5, value=1}]}}}
      CurveIndex=$System.FN_Contrail_Setup.Dive Value
     )
    )
    Module.Index0Param0Write=true
    Module.Index0Param1Write=true
    Module.Index0Param2Write=true
    Module.Index0Param3Write=true
   )
   &SolveForcesAndVelocity(Clamp Velocity=false, Limit Acceleration=false)
   &ScaleColor(
    Scale Alpha=&Lerp_Float(A=5, B=5, Alpha=$System.FN_Contrail_Setup.Dive Value)
    Scale RGB=<1 1 0.1>
    ScaleA=true
    ScaleRGB=true
   )
  ]
  rendererClasses [&NiagaraMeshRendererProperties]
  renderers [{rendererClass=&NiagaraMeshRendererProperties, rendererIndex=0}]
 }
 FireCone002 {
  emitterUpdateScript [
   &EmitterState(Life Cycle Mode=::NewEnumerator1, Loop Behavior=::NewEnumerator1, Loop Duration Mode=::NewEnumerator1, UseLoopDelay=false)
   &SpawnBurst_Instantaneous(Use Spawn Probability=false, UseLoopCountLimit=false)
  ]
  particleSpawnScript [
   &PartitionParticles(Number Of Partitions=9)
   &InitializeParticle(
    Color Mode=::NewEnumerator0
    Lifetime=$Emitter.CurrentLoopDuration
    Mesh Scale Mode=::NewEnumerator3
    Mesh Uniform Scale=2
    Position=$System.SampleSkeletalMesh.SampledPosition
    Position Mode=::NewEnumerator1
    Position Offset=&SelectVectorFromArray(Direct Array Index=$Particles.PartitionParticles.Partition, Vector Selection Array=DataInterface() {propertyValues {FloatData []}})
    Sprite Size Mode=::NewEnumerator3
    Uniform Sprite Size=35
    UsePositionOffset=false
    Write Material Random=true
   )
   &ApplyInitialForces()
   &InitialMeshOrientation(Mesh Orientation Mode=::NewEnumerator7, Use Rotation Vector=true)
   &DynamicMaterialParameters(
    Index 0 Param 2=1
    Index 2 Param 1=1
    Module.Param0WriteEnabled=true
    Module.Param0WriteEnabled2=true
    Module.Param1WriteEnabled=true
    Module.Param1WriteEnabled2=true
    Module.Param2WriteEnabled=true
    Module.Param2WriteEnabled2=true
    Module.Param3WriteEnabled=true
    Module.Param3WriteEnabled2=true
    Write Parameter Index 2=true
   )
  ]
  particleUpdateScript [
   /Game/Effects/Niagara/Modules/Skydiving/FN_DivingAttributes/v6/FN_DivingAttributes()
   &ParticleState()
   &ScaleMeshSize(Scale Factor=&Lerp_Vector(
     A=<1 1 0.5>
     B=<1 1 1>
     Alpha=&FloatFromCurve(
      FloatCurve=DataInterface() {propertyValues {Curve {keys [
          {arriveTangent=4.1785, interpMode=RCIM_Cubic, leaveTangent=4.1785, tangentMode=RCTM_User, time=0, value=0}
          {arriveTangent=1.1179, interpMode=RCIM_Cubic, leaveTangent=1.1179, tangentMode=RCTM_User, time=0.2, value=1}
          {arriveTangent=0.0224, interpMode=RCIM_Cubic, leaveTangent=0.0224, tangentMode=RCTM_User, time=1, value=1}
         ]}}}
      CurveIndex=$System.FN_Contrail_Setup.Dive Value
     )
    ))
   &ScaleColor(
    Scale Alpha=&Lerp_Float(
     A=0.4
     B=0
     Alpha=&FloatFromCurve(
      FloatCurve=DataInterface() {propertyValues {Curve {keys [{tangentMode=RCTM_User, time=0, value=0}, {tangentMode=RCTM_User, time=0.05, value=1}]}}}
      CurveIndex=$System.FN_Contrail_Setup.Dive Value
     )
    )
    Scale RGB=<3 1.5 0.025>
    ScaleA=true
    ScaleRGB=true
   )
   &DynamicMaterialParameters(
    Index 0 Param 1=1
    Index 0 Param 2=0.5
    Index 1 Param 1=&MakeFloatFromVector(
     Vector=&VectorFromFloat(Value=&MakeFloatFromVector(
       Vector=&VectorFromFloat(Value=$System.FN_Contrail_Setup.InterpolateOverTime_Float.Moving Average)
       Channel=::NewEnumerator0
      ))
     Channel=::NewEnumerator0
    )
    Index 1 Param 2=&MakeFloatFromVector(
     Vector=&VectorFromFloat(Value=&MakeFloatFromVector(
       Vector=&VectorFromFloat(Value=$System.FN_Contrail_Setup.InterpolateOverTime_Float.Moving Average)
       Channel=::NewEnumerator1
      ))
     Channel=::NewEnumerator1
    )
    Index 1 Param 3=&MakeFloatFromVector(
     Vector=&VectorFromFloat(Value=$System.FN_Contrail_Setup.InterpolateOverTime_Float.Moving Average)
     Channel=::NewEnumerator2
    )
    Index 2 Param 1=0.5
    Module.Param0WriteEnabled=true
    Module.Param0WriteEnabled1=true
    Module.Param0WriteEnabled2=true
    Module.Param1WriteEnabled=true
    Module.Param1WriteEnabled1=true
    Module.Param1WriteEnabled2=true
    Module.Param2WriteEnabled=true
    Module.Param2WriteEnabled1=true
    Module.Param2WriteEnabled2=true
    Module.Param3WriteEnabled=true
    Module.Param3WriteEnabled1=true
    Module.Param3WriteEnabled2=true
    Write Parameter Index 1=true
    Write Parameter Index 2=true
   ) {module=DynamicMaterialParameters001}
   &SkeletalMeshLocation(
    Bone / Socket Index=4
    Bone Sampling Mode=::NewEnumerator4
    MaskSpawnGroup=false
    Orientation Sampling=::NewEnumerator1
    Sampled Position Offset=<0 0 150>
    Skeletal Mesh=$User.SkeletalMesh
   )
   &UpdateMeshOrientation(
    Facing Direction=&RotateVectorByQuaternion(Vector To Rotate=<0 0 0>, Quaternion=$Particles.SkeletalMeshLocation.SampledRotation)
    Facing Interpolation=false
    Orientation Method=::NewEnumerator1
    Override Quat=false
    TransformFacing=false
    TransformUp=false
    Up Direction:Vector3f=<0 0 0>
   )
   /SetVariables(Particles.Position=&AddVectorToPosition(Position=&Add_Vector(A=$Particles.SkeletalMeshLocation.SampledPosition, B=<0 0 0>), Vector=<0 10 40>)) {isSetParameters=true, module=SetVariables_#2}
   /SetVariables(Particles.Position=&AddVectorToPosition(
     Position=&Add_Vector(
      A=$Particles.SkeletalMeshLocation.SampledPosition
      B=&RotateVectorByQuaternion(
       Vector To Rotate=&Lerp_Vector(
        A=&SelectVectorFromArray(Direct Array Index=$Particles.PartitionParticles.Partition, Direct Array Mode=::NewEnumerator1, Vector Selection Array=$System.Array1)
        B=&SelectVectorFromArray(Direct Array Index=$Particles.PartitionParticles.Partition, Direct Array Mode=::NewEnumerator1, Vector Selection Array=$System.Array2)
        Alpha=&FloatFromCurve(
         FloatCurve=DataInterface() {propertyValues {Curve {keys [
             {arriveTangent=6.8318, interpMode=RCIM_Cubic, leaveTangent=6.8318, tangentMode=RCTM_User, time=0, value=0}
             {arriveTangent=-2.3665, interpMode=RCIM_Cubic, leaveTangent=-2.3665, tangentMode=RCTM_User, time=0.35, value=1.875}
             {arriveTangent=-1.8964, interpMode=RCIM_Cubic, leaveTangent=-1.8964, tangentMode=RCTM_User, time=0.7, value=0.875}
             {interpMode=RCIM_Cubic, tangentMode=RCTM_User, time=0.9, value=1.25}
            ]}}}
         CurveIndex=$System.FN_Contrail_Setup.Dive Value
        )
       )
       Quaternion=&MultiplyQuaternion(Quaternion A=$Particles.SkeletalMeshLocation.SampledRotation, Quaternion B=<0 0 0 1>)
      )
     )
     Vector=<0 0 50>
    )) {enabled=false, isSetParameters=true}
   &SolveForcesAndVelocity(Clamp Velocity=false, Limit Acceleration=false)
   /SetVariables(Particles.VisibilityTag=&MakeIntFromFloat(Float=&Lerp_Float(
      A=0
      B=1
      Alpha=&FloatFromCurve(
       FloatCurve=DataInterface() {propertyValues {Curve {keys [
           {arriveTangent=1.5679, interpMode=RCIM_Cubic, leaveTangent=1.5679, tangentMode=RCTM_User, time=0.8, value=0}
           {arriveTangent=-0.151, interpMode=RCIM_Cubic, leaveTangent=-0.151, tangentMode=RCTM_User, time=0.9, value=1}
           {arriveTangent=0.0224, interpMode=RCIM_Cubic, leaveTangent=0.0224, tangentMode=RCTM_User, time=1, value=1}
          ]}}}
       CurveIndex=$System.FN_Contrail_Setup.Dive Value
      )
     ))) {isSetParameters=true}
   /SetVariables(Particles.MeshOrientation=&LerpQuaternion(
     Quaternion A:Quat4f=&FindQuatBetweenTwoVectors(From Vector=<0 0 0>, To Vector=$Particles.Velocity)
     Quaternion B:Quat4f=&MultiplyQuaternion(
      Quaternion A=$Output.FN_DivingAttributes.Interpolated Rotation
      Quaternion B=&MakeQuatFromAxisAngle(Axis=<1 0 0>, AngleInDegrees=-90)
     )
     Lerp Factor:NiagaraFloat=$System.FN_Contrail_Setup.Dive Value
     Enable Spherical Linear Interpolation:NiagaraBool=true
    )) {isSetParameters=true}
  ]
  rendererClasses [&NiagaraMeshRendererProperties]
  renderers [{rendererClass=&NiagaraMeshRendererProperties, rendererIndex=0}]
 }
 Gizmo_LocalSpace {
  bEnabled=false
  emitterUpdateScript [
   &EmitterState()
   &SpawnBurst_Instantaneous(Spawn Count=4, Spawn Probability=HlslExpression() {hlslExpression="Emitter.LoopCount == 0"}, Use Spawn Probability=true)
  ]
  particleSpawnScript [
   &InitializeParticle(
    Color=<rgba 0 0.3677 1 1>
    Lifetime Max=1
    Lifetime Mode=::NewEnumerator1
    Mesh Scale Mode=::NewEnumerator3
    Mesh Uniform Scale=3
    UsePositionOffset=false
    Write Material Random=true
   )
   &InitialMeshOrientation(Mesh Axis To Orient=<0 0 -1>, Mesh Orientation Mode=::NewEnumerator2, Use Rotation Vector=false)
  ]
  particleUpdateScript [
   &ParticleState(Kill Particles When Lifetime Has Elapsed=false, Let Infinitely Lived Particles Die When Emitter Deactivates=true)
   &UpdateMeshOrientation(
    Facing Direction=$System.FN_Contrail_Setup.Direction Vector
    Facing Interpolation=false
    Orientation Method=::NewEnumerator1
    Override Quat=false
    Stabilization Mode:ENiagara_OrientVectorCount=::NewEnumerator2
    TransformFacing=false
    TransformSide=false
   )
   /SetVariables(Particles.Position=&AddVectorToPosition(
     Position=$Particles.Initial.Position
     Vector=&Multiply_VectorByFloat(
      Vector=$System.FN_Contrail_Setup.Direction Velocity
      Float=&FloatFromCurve(
       FloatCurve=DataInterface() {propertyValues {Curve {keys [{tangentMode=RCTM_User, time=0, value=0}, {tangentMode=RCTM_User, time=1, value=1}]}}}
       Scale Curve=$Particles.Initial.Lifetime
      )
     )
    )) {enabled=false, isSetParameters=true, module=SetVariables_#3}
   /SetVariables(Particles.Velocity=$System.FN_Contrail_Setup.Direction Vector) {isSetParameters=true, module=SetVariables_#4}
  ]
  rendererClasses [&NiagaraMeshRendererProperties]
  renderers [{rendererClass=&NiagaraMeshRendererProperties, rendererIndex=0}]
 }
 Mesh_SkeletalMesh {
  bEnabled=false
  emitterUpdateScript [
   &EmitterState()
   &SpawnBurst_Instantaneous(Spawn Count=4, Spawn Probability=HlslExpression() {hlslExpression="Emitter.LoopCount == 0"}, Use Spawn Probability=true)
  ]
  particleSpawnScript [
   &PartitionParticles(Number Of Partitions=4)
   &InitializeParticle(Color=<rgba 0 1 0.0484 1>, Sprite Size=<10 100>, Sprite Size Mode=::NewEnumerator4, UsePositionOffset=false, Write Material Random=true)
  ]
  particleUpdateScript [
   &ParticleState(Kill Particles When Lifetime Has Elapsed=false, Let Infinitely Lived Particles Die When Emitter Deactivates=true)
   &SkeletalMeshLocation(
    Bone / Socket Index=$Particles.PartitionParticles.Partition
    MaskSpawnGroup=false
    Mesh Sampling Type=::NewEnumerator5
    Orientation Sampling=::NewEnumerator1
    Skeletal Mesh=$User.SkeletalMesh
    Socket Sampling Mode=::NewEnumerator5
   )
   &UpdateMeshOrientation(
    Facing Direction=$System.FN_Contrail_Setup.Direction Vector
    Facing Interpolation=false
    Orientation Method=::NewEnumerator1
    Override Quat=false
    Stabilization Mode:ENiagara_OrientVectorCount=::NewEnumerator2
    TransformFacing=false
    TransformSide=false
   )
   &AlignSpriteToMeshOrientation(Mesh Orientation Relative Sprite Alignment Vector=<-1 0 0>, SetAlignment=true, SetFacing=false)
  ]
  rendererClasses [&NiagaraMeshRendererProperties, &NiagaraSpriteRendererProperties]
  renderers [
   {rendererClass=&NiagaraMeshRendererProperties, rendererIndex=0}
   {rendererClass=&NiagaraSpriteRendererProperties, rendererIndex=1}
  ]
 }
 NumberRibbon002 {
  emitterUpdateScript [
   &EmitterState()
   &SpawnBurst_Instantaneous(Spawn Count=90, Use Spawn Probability=false, UseLoopCountLimit=true)
  ]
  particleSpawnScript [
   &PartitionParticles(
    Normalized Index In Partition Includes One=true
    Number Of Partitions=9
    Particles Per Partition=10
    Partition Mode:ENiagara_PartitionMode=::NewEnumerator1
    Write Partition Index to Ribbon Link Order:ENiagara_AttributeSamplingApplyOutput=::NewEnumerator0
    Write Partition to Ribbon ID=::NewEnumerator0
   )
   &InitializeParticle(
    Color Mode=::NewEnumerator0
    Lifetime=0.3
    Ribbon Width=10
    Ribbon Width Mode=::NewEnumerator1
    Sprite Size Mode=::NewEnumerator3
    UsePositionOffset=false
    Write Material Random=true
   )
  ]
  particleUpdateScript [
   &ParticleState(Kill Particles When Lifetime Has Elapsed=false)
   &SkeletalMeshLocation(
    Bone / Socket Index=4
    Bone Sampling Mode=::NewEnumerator4
    MaskSpawnGroup=false
    Sampled Position Offset=<0 0 150>
    Skeletal Mesh=$User.SkeletalMesh
    Use Sampled Mesh Attribute as Ribbon ID=false
   )
   /SetVariables(Particles.NewAge:NiagaraFloat=&Multiply_Float(A=$Particles.PartitionParticles.Normalized Index In Partition, B=$Particles.Initial.Lifetime)) {isSetParameters=true, module=SetVariables_#5}
   /SetVariables(Particles.Velocity=&Multiply_VectorByFloat(
     Vector=&Lerp_Vector(A=$System.FN_Contrail_Setup.Direction Velocity Raw, B=$System.FN_Contrail_Setup.Direction Velocity, Alpha=$Particles.NewAge)
     Float=$Particles.Initial.Lifetime
    )) {isSetParameters=true, module=SetVariables_#6}
   /SetVariables(Particles.Position=&AddVectorToPosition(
     Position=&Add_Vector(
      A=$Particles.SkeletalMeshLocation.SampledPosition
      B=&RotateVectorByQuaternion(
       Vector To Rotate=&Lerp_Vector(
        A=&SelectVectorFromArray(Direct Array Index=$Particles.PartitionParticles.Partition, Direct Array Mode=::NewEnumerator1, Vector Selection Array=$System.Array1)
        B=&SelectVectorFromArray(Direct Array Index=$Particles.PartitionParticles.Partition, Direct Array Mode=::NewEnumerator1, Vector Selection Array=$System.Array2)
        Alpha=&FloatFromCurve(
         FloatCurve=DataInterface() {propertyValues {Curve {keys [
             {arriveTangent=6.8318, interpMode=RCIM_Cubic, leaveTangent=6.8318, tangentMode=RCTM_User, time=0.075, value=0}
             {arriveTangent=-2.3665, interpMode=RCIM_Cubic, leaveTangent=-2.3665, tangentMode=RCTM_User, time=0.35, value=1.875}
             {arriveTangent=-1.8964, interpMode=RCIM_Cubic, leaveTangent=-1.8964, tangentMode=RCTM_User, time=0.7, value=0.875}
             {interpMode=RCIM_Cubic, tangentMode=RCTM_User, time=0.9, value=1.25}
            ]}}}
         CurveIndex=$System.FN_Contrail_Setup.Dive Value
        )
       )
       Quaternion=&MultiplyQuaternion(Quaternion A=$Particles.SkeletalMeshLocation.SampledRotation, Quaternion B=<0 0 0 1>)
      )
     )
     Vector=&MakeCustomVectorFromBool(
      Bool=$User.bIsFrontEndPreview
      True Vector=&Lerp_Vector(
       A=<0 10 50>
       B=<0 -40 50>
       Alpha=&FloatFromCurve(
        FloatCurve=DataInterface() {propertyValues {Curve {keys [
            {arriveTangent=4.1785, interpMode=RCIM_Cubic, leaveTangent=4.1785, tangentMode=RCTM_User, time=0, value=0}
            {arriveTangent=1.1179, interpMode=RCIM_Cubic, leaveTangent=1.1179, tangentMode=RCTM_User, time=0.2042, value=0.6228}
            {arriveTangent=0.0224, interpMode=RCIM_Cubic, leaveTangent=0.0224, tangentMode=RCTM_User, time=1, value=1}
           ]}}}
        CurveIndex=$System.FN_Contrail_Setup.Dive Value
       )
      )
      False Vector=&Lerp_Vector(
       A=<0 10 50>
       B=<0 50 80>
       Alpha=&FloatFromCurve(
        FloatCurve=DataInterface() {propertyValues {Curve {keys [
            {arriveTangent=4.1785, interpMode=RCIM_Cubic, leaveTangent=4.1785, tangentMode=RCTM_User, time=0, value=0}
            {arriveTangent=1.1179, interpMode=RCIM_Cubic, leaveTangent=1.1179, tangentMode=RCTM_User, time=0.2042, value=0.6228}
            {arriveTangent=0.0224, interpMode=RCIM_Cubic, leaveTangent=0.0224, tangentMode=RCTM_User, time=1, value=1}
           ]}}}
        CurveIndex=$System.FN_Contrail_Setup.Dive Value
       )
      )
     )
    )) {isSetParameters=true, module=SetVariables_#7}
   /SetVariables(Particles.Position=&AddVectorToPosition(
     Position=$Particles.Position
     Vector=&Multiply_VectorByFloat(Vector=$Particles.Velocity, Float=$Particles.PartitionParticles.Normalized Index In Partition)
    )) {isSetParameters=true, module=SetVariables_#8}
   &DynamicMaterialParameters(
    Index 0 Param 1=&RandomRangeFloat(Minimum=1, Maximum=1)
    Index 0 Param 2=&Lerp_Float(A=0, B=1, Alpha=$System.FN_Contrail_Setup.Dive Value)
    Index 1 Param 4=1
    Index 2 Param 1=&MakeCustomFloatFromBool(Bool=&SetBoolByFloatComparison(A=$System.FN_Contrail_Setup.Dive Value, B=0.2), True Float=3, False Float=1)
    Module.Index0Param0Write=true
    Module.Index0Param1Write=true
    Module.Index0Param2Write=true
    Module.Index0Param3Write=true
    Module.Index1Param0Write=true
    Module.Index1Param1Write=true
    Module.Index1Param2Write=true
    Module.Index1Param3Write=true
    Module.Index2Param0Write=true
    Module.Index2Param1Write=true
    Module.Index2Param2Write=true
    Module.Index2Param3Write=true
    Write Parameter Index 1=true
    Write Parameter Index 2=true
   )
   &ScaleColor(
    Scale Alpha=&Multiply_Float(
     A=&Lerp_Float(
      A=0
      B=0.5
      Alpha=&FloatFromCurve(
       FloatCurve=DataInterface() {propertyValues {Curve {keys [
           {arriveTangent=0.5626, interpMode=RCIM_Cubic, leaveTangent=0.5626, tangentMode=RCTM_User, time=0.1, value=0}
           {arriveTangent=1.1179, interpMode=RCIM_Cubic, leaveTangent=1.1179, tangentMode=RCTM_User, time=0.5, value=0.8}
           {arriveTangent=0.0224, interpMode=RCIM_Cubic, leaveTangent=0.0224, tangentMode=RCTM_User, time=1, value=1}
          ]}}}
       CurveIndex=$System.FN_Contrail_Setup.Dive Value
       Scale Curve=&FloatFromCurve(FloatCurve=DataInterface() {propertyValues {Curve {keys [<at 0 0>,<at 0.1 1>]}}}, CurveIndex=$Particles.NewAge)
      )
     )
     B=$System.FN_Contrail_Setup.Normalized Intro Age
    )
    Scale RGB=&MakeVectorFromLinearColorRGB(Color=&LerpLinearColors(
      StartColor=&ColorFromCurve(
       ColorCurve=DataInterface() {propertyValues {
         AlphaCurve {keys [<at 0 1 RCIM_Cubic>]}
         BlueCurve {keys [<at 0 1.0679 RCIM_Cubic>,<at 0.0387 1.0453>,<at 0.0959 0.7054>,<at 1 1 RCIM_Cubic>]}
         GreenCurve {keys [<at 0 4.1811 RCIM_Cubic>,<at 0.0387 3.1601>,<at 0.0959 0.8583>,<at 1 1 RCIM_Cubic>]}
         RedCurve {keys [<at 0 5 RCIM_Cubic>,<at 0.0387 5>,<at 0.0959 2>,<at 1 1 RCIM_Cubic>]}
        }}
       CurveIndex=$Particles.NewAge
      )
      EndColor=&ColorFromCurve(
       ColorCurve=DataInterface() {propertyValues {
         AlphaCurve {keys [<at 0 1 RCIM_Cubic>]}
         BlueCurve {keys [<at 0 5 RCIM_Cubic>, {arriveTangent=-39.5688, interpMode=RCIM_Cubic, leaveTangent=-39.5688, time=0.0518, value=0}, <at 0.1156 0.4254>]}
         GreenCurve {keys [
           <at 0 2.2389 RCIM_Cubic>
           {arriveTangent=-7.4825, interpMode=RCIM_Cubic, leaveTangent=-7.4825, time=0.0518, value=1.8534}
           <at 0.1156 1.3738>
          ]}
         RedCurve {keys [<at 0 1.7795 RCIM_Cubic>, {arriveTangent=10.5567, interpMode=RCIM_Cubic, leaveTangent=10.5567, time=0.0518, value=2}, <at 0.1156 3>]}
        }}
       CurveIndex=$Particles.NewAge
      )
      LerpFactor=$System.FN_Contrail_Setup.Dive Value
     ))
    ScaleA=true
    ScaleRGB=true
   ) {module=ScaleColor001}
   &ScaleRibbonWidth(Ribbon Width Scale=&Multiply_Float(
     A=&Lerp_Float(
      A=1.25
      B=1.5
      Alpha=&FloatFromCurve(
       FloatCurve=DataInterface() {propertyValues {Curve {keys [
           {arriveTangent=4.1785, interpMode=RCIM_Cubic, leaveTangent=4.1785, tangentMode=RCTM_User, time=0, value=0}
           {arriveTangent=1.1179, interpMode=RCIM_Cubic, leaveTangent=1.1179, tangentMode=RCTM_User, time=0.2042, value=0.6228}
           {arriveTangent=0.0224, interpMode=RCIM_Cubic, leaveTangent=0.0224, tangentMode=RCTM_User, time=1, value=1}
          ]}}}
       CurveIndex=$System.FN_Contrail_Setup.Dive Value
      )
     )
     B=&Lerp_Float(
      A=0
      B=1
      Alpha=&FloatFromCurve(
       FloatCurve=DataInterface() {propertyValues {Curve {keys [
           {arriveTangent=4.1785, interpMode=RCIM_Cubic, leaveTangent=4.1785, tangentMode=RCTM_User, time=0, value=0}
           {arriveTangent=0.0079, interpMode=RCIM_Cubic, leaveTangent=0.0079, tangentMode=RCTM_User, time=0.01, value=1}
           {arriveTangent=0.0224, interpMode=RCIM_Cubic, leaveTangent=0.0224, tangentMode=RCTM_User, time=1, value=1}
          ]}}}
       CurveIndex=$System.FN_Contrail_Setup.Dive Value
      )
     )
    )) {module=ScaleRibbonWidth002}
  ]
  rendererClasses [&NiagaraRibbonRendererProperties]
  renderers [{rendererClass=&NiagaraRibbonRendererProperties}]
 }
 Particle_Root {
  bEnabled=false
  emitterUpdateScript [
   &EmitterState()
   &SpawnRate(SpawnRate=31, Use Spawn Probability=false)
  ]
  particleSpawnScript [
   &InitializeParticle(
    Color=<rgba 0 0.3677 1 1>
    Lifetime Max=1
    Lifetime Mode=::NewEnumerator1
    Sprite Size Mode=::NewEnumerator1
    UsePositionOffset=false
    Write Material Random=true
   )
   /SetVariables(Particles.Velocity=$System.FN_Contrail_Setup.Direction Vector) {enabled=false, isSetParameters=true}
   &ShapeLocation(
    Cylinder Height Midpoint=1
    Cylinder Radius=400
    MaskSpawnGroup=false
    Offset Mode=::NewEnumerator0
    Override Seed=false
    Rotation Axis=$System.FN_Contrail_Setup.Direction Vector
    Rotation Mode=::NewEnumerator1
    Shape Primitive=::NewEnumerator1
    Surface Only=true
    Surface Only Band Thickness=0.75
    Transform Order:ENiagara_TransformOrder=::NewEnumerator1
    Use Endcaps In Surface Only Mode=false
    UseOwnerScale=true
   )
  ]
  particleUpdateScript [
   &ParticleState()
   /SetVariables(Particles.NewAge:NiagaraFloat=&Multiply_Float(A=$Particles.Age, B=1)) {isSetParameters=true}
   /SetVariables(Particles.Velocity=&Multiply_VectorByFloat(
     Vector=&Lerp_Vector(A=$System.FN_Contrail_Setup.Direction Vector, B=$System.FN_Contrail_Setup.Direction Velocity, Alpha=$Particles.NewAge)
     Float=1
    )) {isSetParameters=true}
   /SetVariables(Particles.Position=&AddVectorToPosition(
     Position=$Particles.Initial.Position
     Vector=&Multiply_VectorByFloat(Vector=$Particles.Velocity, Float=$Particles.Age)
    )) {isSetParameters=true, module=SetVariables_#3}
   &ScaleSpriteSize(
    Scale Factor=&Lerp_Vector2(A=<1 4>, B=<1 12>, Alpha=$System.FN_Contrail_Setup.Dive Value)
    Scale Sprite Size Mode=::NewEnumerator2
   )
   &ScaleColor(
    Scale Alpha=&FloatFromCurve(
     FloatCurve=DataInterface() {propertyValues {Curve {keys [
         {interpMode=RCIM_Cubic, tangentMode=RCTM_User, time=0, value=0}
         {interpMode=RCIM_Cubic, tangentMode=RCTM_User, time=0.5, value=1}
         {tangentMode=RCTM_User, time=1, value=0}
        ]}}}
     Scale Curve=$System.FN_Contrail_Setup.Normalized Intro Age
    )
    ScaleA=true
    ScaleRGB=true
   )
  ]
  rendererClasses [&NiagaraSpriteRendererProperties, &NiagaraMeshRendererProperties]
  renderers [
   {rendererClass=&NiagaraSpriteRendererProperties}
   {rendererClass=&NiagaraMeshRendererProperties, rendererIndex=1}
  ]
 }
 Particles_SkeletalMesh {
  bEnabled=false
  emitterUpdateScript [
   &EmitterState()
   &SpawnRate(SpawnRate=40, Use Spawn Probability=false)
  ]
  particleSpawnScript [
   &PartitionParticles(Number Of Partitions=4)
   &InitializeParticle(
    Color Mode=::NewEnumerator0
    Lifetime Max=1
    Lifetime Mode=::NewEnumerator1
    Sprite Size=<5 20>
    Sprite Size Mode=::NewEnumerator4
    UsePositionOffset=false
    Write Material Random=true
   )
   &ShapeLocation(
    Apply To Particle Position=::NewEnumerator1
    MaskSpawnGroup=false
    Override Seed=false
    Shape Origin=&VectorFromFloat(Value=0)
    Sphere Radius=5
    UseOwnerScale=true
   )
  ]
  particleUpdateScript [
   &ParticleState()
   &SkeletalMeshLocation(
    Bone / Socket Index=$Particles.PartitionParticles.Partition
    MaskSpawnGroup=false
    Mesh Sampling Type=::NewEnumerator5
    Orientation Sampling=::NewEnumerator1
    Sampled Position Offset=$Particles.ShapeLocation.ShapePosition
    Skeletal Mesh=$User.SkeletalMesh
    Socket Sampling Mode=::NewEnumerator5
   )
   /SetVariables(Particles.NewAge:NiagaraFloat=&Multiply_Float(A=$Particles.Age, B=0.75)) {isSetParameters=true}
   /SetVariables(Particles.Velocity=&Multiply_VectorByFloat(Vector=&Lerp_Vector(A=<0 0 0>, B=$System.FN_Contrail_Setup.Direction Velocity, Alpha=$Particles.NewAge), Float=1)) {isSetParameters=true}
   /SetVariables(Particles.Position=&AddVectorToPosition(
     Position=$Particles.Position
     Vector=&Multiply_VectorByFloat(Vector=$Particles.Velocity, Float=$Particles.Age)
    )) {isSetParameters=true, module=SetVariables_#3}
   /SetVariables(Particles.Velocity=$System.FN_Contrail_Setup.Direction Vector) {enabled=false, isSetParameters=true, module=SetVariables_#4}
   &ScaleColor(
    Scale Alpha=&FloatFromCurve(
     FloatCurve=DataInterface() {propertyValues {Curve {keys [
         {interpMode=RCIM_Cubic, tangentMode=RCTM_User, time=0, value=0}
         {interpMode=RCIM_Cubic, tangentMode=RCTM_User, time=0.5, value=1}
         {tangentMode=RCTM_User, time=1, value=0}
        ]}}}
     Scale Curve=$System.FN_Contrail_Setup.Normalized Intro Age
    )
    Scale RGB=<0.3 0.3 0.3>
    ScaleA=true
    ScaleRGB=true
   )
  ]
  rendererClasses [&NiagaraSpriteRendererProperties]
  renderers [{rendererClass=&NiagaraSpriteRendererProperties}]
 }
 ParticletoRibbon_SkeletalMesh {
  bEnabled=false
  emitterUpdateScript [
   &EmitterState()
   &SpawnBurst_Instantaneous(Spawn Count=40, Use Spawn Probability=false, UseLoopCountLimit=true)
  ]
  particleSpawnScript [
   &PartitionParticles(
    Normalized Index In Partition Includes One=true
    Number Of Partitions=4
    Particles Per Partition=10
    Partition Mode:ENiagara_PartitionMode=::NewEnumerator1
    Write Partition Index to Ribbon Link Order:ENiagara_AttributeSamplingApplyOutput=::NewEnumerator0
    Write Partition to Ribbon ID=::NewEnumerator0
   )
   &InitializeParticle(
    Color Mode=::NewEnumerator0
    Lifetime=0.5
    Ribbon Width=3
    Ribbon Width Mode=::NewEnumerator1
    Sprite Size Mode=::NewEnumerator3
    UsePositionOffset=false
    Write Material Random=true
   )
  ]
  particleUpdateScript [
   &ParticleState(Kill Particles When Lifetime Has Elapsed=false)
   &SkeletalMeshLocation(
    Bone / Socket Index=$Particles.PartitionParticles.Partition
    MaskSpawnGroup=false
    Mesh Sampling Type=::NewEnumerator5
    Skeletal Mesh=$User.SkeletalMesh
    Socket Sampling Mode=::NewEnumerator5
   )
   /SetVariables(Particles.NewAge:NiagaraFloat=&Multiply_Float(A=$Particles.PartitionParticles.Normalized Index In Partition, B=$Particles.Initial.Lifetime)) {isSetParameters=true, module=SetVariables_#5}
   /SetVariables(Particles.Velocity=&Multiply_VectorByFloat(
     Vector=&Lerp_Vector(A=$System.FN_Contrail_Setup.Direction Velocity Raw, B=$System.FN_Contrail_Setup.Direction Velocity, Alpha=$Particles.NewAge)
     Float=$Particles.Initial.Lifetime
    )) {isSetParameters=true, module=SetVariables_#6}
   /SetVariables(Particles.Position=&AddVectorToPosition(
     Position=$Particles.Position
     Vector=&Multiply_VectorByFloat(Vector=$Particles.Velocity, Float=$Particles.PartitionParticles.Normalized Index In Partition)
    )) {isSetParameters=true, module=SetVariables_#3}
   &ScaleColor(Scale Alpha=$System.FN_Contrail_Setup.Normalized Intro Age, Scale RGB=<0.1 0.1 0.1>, ScaleA=true, ScaleRGB=true)
  ]
  rendererClasses [&NiagaraSpriteRendererProperties, &NiagaraRibbonRendererProperties]
  renderers [
   {rendererClass=&NiagaraSpriteRendererProperties}
   {rendererClass=&NiagaraRibbonRendererProperties, rendererIndex=1}
  ]
 }
 SpeedRibbon001 {
  emitterUpdateScript [
   &EmitterState()
   &SpawnBurst_Instantaneous(Spawn Count=90, Use Spawn Probability=true, UseLoopCountLimit=true)
  ]
  particleSpawnScript [
   &PartitionParticles(
    Normalized Index In Partition Includes One=true
    Number Of Partitions=9
    Particles Per Partition=10
    Partition Mode:ENiagara_PartitionMode=::NewEnumerator1
    Write Partition Index to Ribbon Link Order:ENiagara_AttributeSamplingApplyOutput=::NewEnumerator0
    Write Partition to Ribbon ID=::NewEnumerator0
   )
   &InitializeParticle(
    Color Mode=::NewEnumerator0
    Lifetime=0.2
    Ribbon Width=10
    Ribbon Width Mode=::NewEnumerator1
    Sprite Size Mode=::NewEnumerator3
    UsePositionOffset=false
    Write Material Random=true
   )
  ]
  particleUpdateScript [
   &ParticleState(Kill Particles When Lifetime Has Elapsed=false)
   &SkeletalMeshLocation(
    Bone / Socket Index=4
    Bone Sampling Mode=::NewEnumerator4
    MaskSpawnGroup=false
    Sampled Position Offset=<0 0 150>
    Skeletal Mesh=$User.SkeletalMesh
    Use Sampled Mesh Attribute as Ribbon ID=false
   )
   /SetVariables(Particles.NewAge:NiagaraFloat=&Multiply_Float(A=$Particles.PartitionParticles.Normalized Index In Partition, B=$Particles.Initial.Lifetime)) {isSetParameters=true, module=SetVariables_#5}
   /SetVariables(Particles.Velocity=&Multiply_VectorByFloat(
     Vector=&Lerp_Vector(A=$System.FN_Contrail_Setup.Direction Velocity Raw, B=$System.FN_Contrail_Setup.Direction Velocity, Alpha=$Particles.NewAge)
     Float=$Particles.Initial.Lifetime
    )) {isSetParameters=true, module=SetVariables_#6}
   /SetVariables(Particles.Position=&AddVectorToPosition(
     Position=&Add_Vector(
      A=$Particles.SkeletalMeshLocation.SampledPosition
      B=&RotateVectorByQuaternion(
       Vector To Rotate=&Lerp_Vector(
        A=&SelectVectorFromArray(Direct Array Index=$Particles.PartitionParticles.Partition, Direct Array Mode=::NewEnumerator1, Vector Selection Array=$System.Array1)
        B=&SelectVectorFromArray(Direct Array Index=$Particles.PartitionParticles.Partition, Direct Array Mode=::NewEnumerator1, Vector Selection Array=$System.Array2)
        Alpha=&FloatFromCurve(
         FloatCurve=DataInterface() {propertyValues {Curve {keys [
             {arriveTangent=6.8318, interpMode=RCIM_Cubic, leaveTangent=6.8318, tangentMode=RCTM_User, time=0.075, value=0}
             {arriveTangent=-2.3665, interpMode=RCIM_Cubic, leaveTangent=-2.3665, tangentMode=RCTM_User, time=0.35, value=1.875}
             {arriveTangent=-1.8964, interpMode=RCIM_Cubic, leaveTangent=-1.8964, tangentMode=RCTM_User, time=0.7, value=0.875}
             {interpMode=RCIM_Cubic, tangentMode=RCTM_User, time=0.9, value=1.25}
            ]}}}
         CurveIndex=$System.FN_Contrail_Setup.Dive Value
        )
       )
       Quaternion=&MultiplyQuaternion(Quaternion A=$Particles.SkeletalMeshLocation.SampledRotation, Quaternion B=<0 0 0 1>)
      )
     )
     Vector=&MakeCustomVectorFromBool(
      Bool=$User.bIsFrontEndPreview
      True Vector=&Lerp_Vector(
       A=<0 10 50>
       B=<0 -40 50>
       Alpha=&FloatFromCurve(
        FloatCurve=DataInterface() {propertyValues {Curve {keys [
            {arriveTangent=4.1785, interpMode=RCIM_Cubic, leaveTangent=4.1785, tangentMode=RCTM_User, time=0, value=0}
            {arriveTangent=1.1179, interpMode=RCIM_Cubic, leaveTangent=1.1179, tangentMode=RCTM_User, time=0.2042, value=0.6228}
            {arriveTangent=0.0224, interpMode=RCIM_Cubic, leaveTangent=0.0224, tangentMode=RCTM_User, time=1, value=1}
           ]}}}
        CurveIndex=$System.FN_Contrail_Setup.Dive Value
       )
      )
      False Vector=&Lerp_Vector(
       A=<0 10 50>
       B=<0 50 80>
       Alpha=&FloatFromCurve(
        FloatCurve=DataInterface() {propertyValues {Curve {keys [
            {arriveTangent=4.1785, interpMode=RCIM_Cubic, leaveTangent=4.1785, tangentMode=RCTM_User, time=0, value=0}
            {arriveTangent=1.1179, interpMode=RCIM_Cubic, leaveTangent=1.1179, tangentMode=RCTM_User, time=0.2042, value=0.6228}
            {arriveTangent=0.0224, interpMode=RCIM_Cubic, leaveTangent=0.0224, tangentMode=RCTM_User, time=1, value=1}
           ]}}}
        CurveIndex=$System.FN_Contrail_Setup.Dive Value
       )
      )
     )
    )) {isSetParameters=true, module=SetVariables_#7}
   /SetVariables(Particles.Position=&AddVectorToPosition(
     Position=$Particles.Position
     Vector=&Multiply_VectorByFloat(Vector=$Particles.Velocity, Float=$Particles.PartitionParticles.Normalized Index In Partition)
    )) {isSetParameters=true, module=SetVariables_#8}
   &DynamicMaterialParameters(
    Index 0 Param 1=&RandomRangeFloat(Minimum=1, Maximum=1)
    Index 0 Param 2=&Lerp_Float(A=0.5, B=1, Alpha=$System.FN_Contrail_Setup.Dive Value)
    Index 2 Param 1=&Lerp_Float(
     A=1
     B=1
     Alpha=&FloatFromCurve(
      FloatCurve=DataInterface() {propertyValues {Curve {keys [
          {arriveTangent=4.1785, interpMode=RCIM_Cubic, leaveTangent=4.1785, tangentMode=RCTM_User, time=0, value=0}
          {arriveTangent=1.1179, interpMode=RCIM_Cubic, leaveTangent=1.1179, tangentMode=RCTM_User, time=0.2042, value=0.6228}
          {arriveTangent=0.0224, interpMode=RCIM_Cubic, leaveTangent=0.0224, tangentMode=RCTM_User, time=1, value=1}
         ]}}}
      CurveIndex=$System.FN_Contrail_Setup.Dive Value
     )
    )
    Module.Index0Param0Write=true
    Module.Index0Param1Write=true
    Module.Index0Param2Write=true
    Module.Index0Param3Write=true
    Module.Index2Param0Write=true
    Module.Index2Param1Write=true
    Module.Index2Param2Write=true
    Module.Index2Param3Write=true
    Write Parameter Index 2=true
   )
   &ScaleColor(
    Scale Alpha=&Multiply_Float(
     A=&Lerp_Float(
      A=0
      B=1
      Alpha=&FloatFromCurve(
       FloatCurve=DataInterface() {propertyValues {Curve {keys [
           {arriveTangent=0.5626, interpMode=RCIM_Cubic, leaveTangent=0.5626, tangentMode=RCTM_User, time=0.1, value=0}
           {arriveTangent=1.1179, interpMode=RCIM_Cubic, leaveTangent=1.1179, tangentMode=RCTM_User, time=0.5, value=0.8}
           {arriveTangent=0.0224, interpMode=RCIM_Cubic, leaveTangent=0.0224, tangentMode=RCTM_User, time=1, value=1}
          ]}}}
       CurveIndex=$System.FN_Contrail_Setup.Dive Value
       Scale Curve=&FloatFromCurve(FloatCurve=DataInterface() {propertyValues {Curve {keys [<at 0 0>,<at 0.1 1>]}}}, CurveIndex=$Particles.NewAge)
      )
     )
     B=$System.FN_Contrail_Setup.Normalized Intro Age
    )
    Scale RGB=&MakeVectorFromLinearColorRGB(Color=&LerpLinearColors(
      StartColor=&ColorFromCurve(
       ColorCurve=DataInterface() {propertyValues {
         AlphaCurve {keys [<at 0 1 RCIM_Cubic>]}
         BlueCurve {keys [<at 0 1.0679 RCIM_Cubic>,<at 0.0387 1.0453>,<at 0.0959 0.7054>,<at 1 1 RCIM_Cubic>]}
         GreenCurve {keys [<at 0 4.1811 RCIM_Cubic>,<at 0.0387 3.1601>,<at 0.0959 0.8583>,<at 1 1 RCIM_Cubic>]}
         RedCurve {keys [<at 0 5 RCIM_Cubic>,<at 0.0387 5>,<at 0.0959 2>,<at 1 1 RCIM_Cubic>]}
        }}
       CurveIndex=$Particles.NewAge
      )
      EndColor=&ColorFromCurve(
       ColorCurve=DataInterface() {propertyValues {
         AlphaCurve {keys [<at 0 1 RCIM_Cubic>]}
         BlueCurve {keys [<at 0 5 RCIM_Cubic>, {arriveTangent=-39.5688, interpMode=RCIM_Cubic, leaveTangent=-39.5688, time=0.0518, value=0}, <at 0.1156 0.4254>]}
         GreenCurve {keys [
           <at 0 2.2389 RCIM_Cubic>
           {arriveTangent=-7.4825, interpMode=RCIM_Cubic, leaveTangent=-7.4825, time=0.0518, value=1.8534}
           <at 0.1156 1.3738>
          ]}
         RedCurve {keys [<at 0 1.7795 RCIM_Cubic>, {arriveTangent=10.5567, interpMode=RCIM_Cubic, leaveTangent=10.5567, time=0.0518, value=2}, <at 0.1156 3>]}
        }}
       CurveIndex=$Particles.NewAge
      )
      LerpFactor=$System.FN_Contrail_Setup.Dive Value
     ))
    ScaleA=true
    ScaleRGB=true
   ) {module=ScaleColor002}
   &ScaleRibbonWidth(Ribbon Width Scale=&Multiply_Float(
     A=&Lerp_Float(
      A=1.25
      B=1.5
      Alpha=&FloatFromCurve(
       FloatCurve=DataInterface() {propertyValues {Curve {keys [
           {arriveTangent=4.1785, interpMode=RCIM_Cubic, leaveTangent=4.1785, tangentMode=RCTM_User, time=0, value=0}
           {arriveTangent=1.1179, interpMode=RCIM_Cubic, leaveTangent=1.1179, tangentMode=RCTM_User, time=0.2042, value=0.6228}
           {arriveTangent=0.0224, interpMode=RCIM_Cubic, leaveTangent=0.0224, tangentMode=RCTM_User, time=1, value=1}
          ]}}}
       CurveIndex=$System.FN_Contrail_Setup.Dive Value
      )
     )
     B=&Lerp_Float(
      A=0
      B=1
      Alpha=&FloatFromCurve(
       FloatCurve=DataInterface() {propertyValues {Curve {keys [
           {arriveTangent=4.1785, interpMode=RCIM_Cubic, leaveTangent=4.1785, tangentMode=RCTM_User, time=0, value=0}
           {arriveTangent=0.0079, interpMode=RCIM_Cubic, leaveTangent=0.0079, tangentMode=RCTM_User, time=0.01, value=1}
           {arriveTangent=0.0224, interpMode=RCIM_Cubic, leaveTangent=0.0224, tangentMode=RCTM_User, time=1, value=1}
          ]}}}
       CurveIndex=$System.FN_Contrail_Setup.Dive Value
      )
     )
    )) {module=ScaleRibbonWidth004}
   &ScaleRibbonWidth(Ribbon Width Scale=&FloatFromCurve(
     FloatCurve=DataInterface() {propertyValues {Curve {keys [
         {arriveTangent=4.1785, interpMode=RCIM_Cubic, leaveTangent=4.1785, tangentMode=RCTM_User, time=0, value=0.75}
         {arriveTangent=10.0966, interpMode=RCIM_Cubic, leaveTangent=10.0966, tangentMode=RCTM_User, time=0.0205, value=0.8581}
         {arriveTangent=0.0224, interpMode=RCIM_Cubic, leaveTangent=0.0224, tangentMode=RCTM_User, time=0.05, value=1}
        ]}}}
     CurveIndex=&Multiply_Float(A=$Particles.PartitionParticles.Normalized Index In Partition, B=$Particles.Initial.Lifetime)
    )) {module=ScaleRibbonWidth003}
  ]
  rendererClasses [&NiagaraRibbonRendererProperties]
  renderers [{rendererClass=&NiagaraRibbonRendererProperties}]
 }
 Velocity_Cone {
  emitterUpdateScript [
   &EmitterState(Life Cycle Mode=::NewEnumerator1, Loop Behavior=::NewEnumerator1, Loop Duration Mode=::NewEnumerator1, UseLoopDelay=false)
   &SpawnBurst_Instantaneous(Spawn Count=9, Use Spawn Probability=false, UseLoopCountLimit=false)
  ]
  particleSpawnScript [
   &PartitionParticles(Number Of Partitions=9)
   &InitializeParticle(
    Color Mode=::NewEnumerator0
    Lifetime=$Emitter.CurrentLoopDuration
    Mesh Scale Mode=::NewEnumerator3
    Position=$System.SampleSkeletalMesh.SampledPosition
    Position Mode=::NewEnumerator1
    Position Offset=&SelectVectorFromArray(Direct Array Index=$Particles.PartitionParticles.Partition, Vector Selection Array=DataInterface() {propertyValues {FloatData []}})
    Sprite Size Mode=::NewEnumerator3
    Uniform Sprite Size=35
    UsePositionOffset=false
    Write Material Random=true
   )
   &ApplyInitialForces()
   &InitialMeshOrientation(
    Mesh Orientation Mode=::NewEnumerator2
    Orientation Vector=/Game/Effects/Niagara/DynamicInputs/FrontEnd/FN_FrontEndRotateVector(Angle:NiagaraFloat=0.6, Vector to Rotate:Vector3f=<0 0 -1>)
    Use Rotation Vector=false
   ) {enabled=false}
   &InitialMeshOrientation(
    Mesh Orientation Mode=::NewEnumerator2
    Orientation Vector=&MakeCustomVectorFromBool(
     Bool=$User.bIsFrontEndPreview
     True Vector=&FN_FrontEndRotateVector(Angle:NiagaraFloat=0.6, Vector to Rotate:Vector3f=<0 0 -1>)
     False Vector=<0 0 -1>
    )
    Use Rotation Vector=false
   ) {module=InitialMeshOrientation001}
  ]
  particleUpdateScript [
   &MeshRotationForce(Rotation=&Lerp_Vector(A=<0 0 0>, B=<0 50 0>, Alpha=$System.FN_Contrail_Setup.Dive Value)) {enabled=false, module=MeshRotationForce001}
   &ParticleState()
   &SkeletalMeshLocation(
    Bone / Socket Index=4
    Bone Sampling Mode=::NewEnumerator4
    MaskSpawnGroup=false
    Orientation Sampling=::NewEnumerator1
    Sampled Position Offset=<0 0 150>
    Skeletal Mesh=$User.SkeletalMesh
   )
   /SetVariables(Particles.Position=&AddVectorToPosition(
     Position=&Add_Vector(
      A=$Particles.SkeletalMeshLocation.SampledPosition
      B=&RotateVectorByQuaternion(
       Vector To Rotate=&Lerp_Vector(
        A=&SelectVectorFromArray(Direct Array Index=$Particles.PartitionParticles.Partition, Direct Array Mode=::NewEnumerator1, Vector Selection Array=$System.Array1)
        B=&SelectVectorFromArray(Direct Array Index=$Particles.PartitionParticles.Partition, Direct Array Mode=::NewEnumerator1, Vector Selection Array=$System.Array2)
        Alpha=&FloatFromCurve(
         FloatCurve=DataInterface() {propertyValues {Curve {keys [
             {arriveTangent=6.8318, interpMode=RCIM_Cubic, leaveTangent=6.8318, tangentMode=RCTM_User, time=0, value=0}
             {arriveTangent=-2.3665, interpMode=RCIM_Cubic, leaveTangent=-2.3665, tangentMode=RCTM_User, time=0.35, value=1.875}
             {arriveTangent=-1.8964, interpMode=RCIM_Cubic, leaveTangent=-1.8964, tangentMode=RCTM_User, time=0.7, value=0.875}
             {interpMode=RCIM_Cubic, tangentMode=RCTM_User, time=0.9, value=1.25}
            ]}}}
         CurveIndex=$System.FN_Contrail_Setup.Dive Value
        )
       )
       Quaternion=&MultiplyQuaternion(Quaternion A=$Particles.SkeletalMeshLocation.SampledRotation, Quaternion B=<0 0 0 1>)
      )
     )
     Vector=&MakeCustomVectorFromBool(
      Bool=$User.bIsFrontEndPreview
      True Vector=&Lerp_Vector(
       A=<0 10 50>
       B=<0 -40 50>
       Alpha=&FloatFromCurve(
        FloatCurve=DataInterface() {propertyValues {Curve {keys [
            {arriveTangent=4.1785, interpMode=RCIM_Cubic, leaveTangent=4.1785, tangentMode=RCTM_User, time=0, value=0}
            {arriveTangent=1.1179, interpMode=RCIM_Cubic, leaveTangent=1.1179, tangentMode=RCTM_User, time=0.2042, value=0.6228}
            {arriveTangent=0.0224, interpMode=RCIM_Cubic, leaveTangent=0.0224, tangentMode=RCTM_User, time=1, value=1}
           ]}}}
        CurveIndex=$System.FN_Contrail_Setup.Dive Value
       )
      )
      False Vector=&Lerp_Vector(
       A=<0 10 50>
       B=<0 50 80>
       Alpha=&FloatFromCurve(
        FloatCurve=DataInterface() {propertyValues {Curve {keys [
            {arriveTangent=4.1785, interpMode=RCIM_Cubic, leaveTangent=4.1785, tangentMode=RCTM_User, time=0, value=0}
            {arriveTangent=1.1179, interpMode=RCIM_Cubic, leaveTangent=1.1179, tangentMode=RCTM_User, time=0.2042, value=0.6228}
            {arriveTangent=0.0224, interpMode=RCIM_Cubic, leaveTangent=0.0224, tangentMode=RCTM_User, time=1, value=1}
           ]}}}
        CurveIndex=$System.FN_Contrail_Setup.Dive Value
       )
      )
     )
    )) {isSetParameters=true, module=SetVariables_#2}
   &DynamicMaterialParameters(
    Index 0 Param 1=&RandomRangeFloat(Minimum=0, Maximum=1)
    Index 0 Param 2=1
    Index 0 Param 3=&Lerp_Float(
     A=0.25
     B=0.5
     Alpha=&FloatFromCurve(
      FloatCurve=DataInterface() {propertyValues {Curve {keys [{interpMode=RCIM_Cubic, tangentMode=RCTM_Break, time=0, value=0}, {interpMode=RCIM_Cubic, tangentMode=RCTM_Break, time=0.5, value=1}]}}}
      CurveIndex=$User.ParaGlideLeanAlpha
     )
    )
    Module.Index0Param0Write=true
    Module.Index0Param1Write=true
    Module.Index0Param2Write=true
    Module.Index0Param3Write=true
   )
   &UpdateMeshOrientation(
    Facing Direction=&RotateVectorByQuaternion(
     Vector To Rotate=<0 0 0>
     Quaternion=&FindQuatBetweenTwoVectors(From Vector=$Particles.SkeletalMeshLocation.SampledPosition, To Vector=$Particles.SkeletalMeshLocation.SampledVelocity)
    )
    Facing Interpolation=false
    Orientation Method=::NewEnumerator1
    Override Quat=false
    TransformFacing=false
    TransformUp=false
   ) {module=UpdateMeshOrientation001}
   &SolveForcesAndVelocity(Clamp Velocity=false, Limit Acceleration=false)
   &ScaleColor(
    Scale Alpha=&Lerp_Float(
     A=0
     B=0.1
     Alpha=&FloatFromCurve(
      FloatCurve=DataInterface() {propertyValues {Curve {keys [<at 0.25 0 RCIM_Cubic>, {arriveTangent=1.3333, interpMode=RCIM_Cubic, leaveTangent=1.3333, time=0.65, value=0.15}, <at 1 1 RCIM_Cubic>]}}}
      CurveIndex=$System.FN_Contrail_Setup.Dive Value
     )
    )
    Scale RGB=&MakeVectorFromLinearColorRGB(Color=&LerpLinearColors(StartColor=<rgba 1 0.8388 0 1>, EndColor=<rgba 1.0697 3.5196 5 1>, LerpFactor=$System.FN_Contrail_Setup.Dive Value))
    ScaleA=true
    ScaleRGB=true
   )
  ]
  rendererClasses [&NiagaraMeshRendererProperties]
  renderers [{rendererClass=&NiagaraMeshRendererProperties, rendererIndex=0}]
 }
}
summary {
 systemName=NS_Contrail_StageCue
 emitters [
  {
   bEnabled=false
   emitterName=Mesh_SkeletalMesh
   rendererClasses [&NiagaraMeshRendererProperties, &NiagaraSpriteRendererProperties]
  }
  {
   bEnabled=false
   emitterName=Particle_Root
   rendererClasses [&NiagaraSpriteRendererProperties, &NiagaraMeshRendererProperties]
  }
  {bEnabled=false, emitterName=Particles_SkeletalMesh, rendererClasses [&NiagaraSpriteRendererProperties]}
  {bEnabled=false, emitterName=Gizmo_LocalSpace, rendererClasses [&NiagaraMeshRendererProperties]}
  {
   bEnabled=false
   emitterName=ParticletoRibbon_SkeletalMesh
   rendererClasses [&NiagaraSpriteRendererProperties, &NiagaraRibbonRendererProperties]
  }
  {emitterName=CueBalls, rendererClasses [&NiagaraMeshRendererProperties]}
  {emitterName=Velocity_Cone, rendererClasses [&NiagaraMeshRendererProperties]}
  {emitterName=SpeedRibbon001, rendererClasses [&NiagaraRibbonRendererProperties]}
  {emitterName=NumberRibbon002, rendererClasses [&NiagaraRibbonRendererProperties]}
  {emitterName=FireCone002, rendererClasses [&NiagaraMeshRendererProperties]}
 ]
 userVariables [
  User.bIsFrontEnd {type=NiagaraBool}
  User.bIsFrontEndPreview {type=NiagaraBool, defaultValue=true}
  User.isFrontend {type=NiagaraFloat, defaultValue=0, description=""}
  User.ParaGlideLeanAlpha {type=NiagaraFloat, defaultValue=0, description=""}
  User.PlayerSpeed {type=NiagaraFloat, defaultValue=0, description=""}
  User.RibbonIDOffset {type=NiagaraInt32}
  User.SkeletalMesh {
   type=NiagaraDataInterfaceSkeletalMesh
   defaultValue {
    struct=NiagaraExt_VariableValue_DataInterface
    value {dataInterface=&self:NiagaraDataInterfaceSkeletalMesh_0}
   }
  }
 ]
}
